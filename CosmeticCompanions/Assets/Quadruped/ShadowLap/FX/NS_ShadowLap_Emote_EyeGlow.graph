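# Niagara system that draws glowing eyes for an emote of the ShadowLap quadruped cosmetic companion.
# One burst spawns two particles, partitioned by index into mirrored eye offsets and tilts, each rendered by a sprite and two mesh renderers.
# Alpha fades in over system age, multiplied by a smoothed inverse of User.BlinkAlpha_R or _L, so each eye blinks independently.
# A derived Color_SpotLight attribute carries the same color at reduced alpha.
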NiagaraSystem /CosmeticCompanions/Assets/Quadruped/ShadowLap/FX/NS_ShadowLap_Emote_EyeGlow
dependencies {usedDataInterfaces [&NiagaraDataInterfaceCurve], usedDynamicInputs [], usedModules [&SystemState]}
emitters {EyeGlows {
  emitterUpdateScript [
   &EmitterState(
    Inactive Response=::NewEnumerator1
    Life Cycle Mode=::NewEnumerator1
    Loop Behavior=::NewEnumerator1
    Loop Duration Mode=::NewEnumerator1
    UseLoopDelay=false
   )
   &SpawnBurst_Instantaneous(Spawn Count=2, Use Spawn Probability=false, UseLoopCountLimit=true)
  ]
  particleSpawnScript [
   &PartitionParticles(ID to Partition:NiagaraInt32=&ReturnExecIndex())
   &InitializeParticle(
    Mesh Scale=<0.25 0.25 0.15>
    Mesh Scale Mode=::NewEnumerator4
    Position Offset=&MergeHelper_VectorSwitch(0=<-5.7015 8.8863 6.9869>, 1=<5.9186 8.8868 6.9094>, 2=?, 3=?, Index=$Particles.PartitionParticles.Partition, Num=::NewEnumerator0)
    Sprite Size Mode=::NewEnumerator3
    Uniform Sprite Size=30
    UsePositionOffset=true
    Write Material Random=true
   )
   &CameraOffset(Camera Offset Amount=4)
   &InitialMeshOrientation(
    Mesh Axis To Orient=<0 0 1>
    Mesh Orientation Mode=::NewEnumerator2
    Orientation Vector=<0 -1 0>
    Rotation=&MergeHelper_VectorSwitch(0=<0.1 0 0.1>, 1=<0.1 0 -0.1>, 2=?, 3=?, Index=$Particles.PartitionParticles.Partition, Num=::NewEnumerator0)
    Use Rotation Vector=true
   )
   &DynamicMaterialParameters(
    Index 0 Param 1=0.25
    Module.Index0Float0Write=true
    Module.Index0Float1Write=false
    Module.Index0Float2Write=false
    Module.Index0Float3Write=false
   )
  ]
  particleUpdateScript [
   &ParticleState(Kill Particles When Lifetime Has Elapsed=false, Let Infinitely Lived Particles Die When Emitter Deactivates=true)
   &ScaleColor(
    Scale Alpha=&Multiply_Float(
     A=&FloatFromCurve(
      FloatCurve=DataInterface() {propertyValues {Curve {
         defaultValue=1
         keys [
          {arriveTangent=3.7487, interpMode=RCIM_Cubic, leaveTangent=3.7487, tangentMode=RCTM_User, time=0, value=0}
          {interpMode=RCIM_Cubic, tangentMode=RCTM_User, time=1, value=1}
         ]
        }}}
      CurveIndex=$Engine.System.Age
     )
     B=/Niagara/DynamicInputs/Lerp/SmoothLerpOverTime_Float(
      Convergence Rate:NiagaraFloat=3
      Smooth Value:NiagaraFloat=&OneMinusFloat(Float=&MergeHelper_FloatSwitch(0=$User.BlinkAlpha_R, 1=$User.BlinkAlpha_L, 2=?, 3=?, Index=$Particles.PartitionParticles.Partition, Num=::NewEnumerator0))
      Target Value:NiagaraFloat=1
     )
    )
    ScaleA=true
    ScaleRGB=true
   )
   /SetVariables(Particles.Color_SpotLight:LinearColor=&MultiplyLinearColor(
     Linear Color=$Particles.Color
     Scale Factor=&MakeVector4FromVectorAndFloat("Vector (RGB)"=<1 1 1>, "Float (Alpha)"=0.25)
    )) {isSetParameters=true}
  ]
  rendererClasses [&NiagaraSpriteRendererProperties, &NiagaraMeshRendererProperties]
  renderers [
   {rendererClass=&NiagaraSpriteRendererProperties}
   {rendererClass=&NiagaraMeshRendererProperties, rendererIndex=1}
   {rendererClass=&NiagaraMeshRendererProperties, rendererIndex=2}
  ]
 }}
summary {
 systemName=NS_ShadowLap_Emote_EyeGlow
 emitters [{emitterName=EyeGlows, rendererClasses [&NiagaraSpriteRendererProperties, &NiagaraMeshRendererProperties]}]
 userVariables [
  User.BlinkAlpha_L {type=NiagaraFloat, defaultValue=0, description=""}
  User.BlinkAlpha_R {type=NiagaraFloat, defaultValue=0, description=""}
 ]
}
