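# Material function for ShadowLap quadruped companion's geometry-flipbook eye expressions, swapping vertex-color-tagged pose meshes.
# Animation-driven curve scalars feed a thresholded If chain of VertexColorMask calls, with a local-X sign test making one pose side-specific.
# Under a static switch, World Position Offset collapses masked vertices to the object position, or applies a debug grid offset.
# Base color above a threshold turns black with white emissive, scaled by time of day.
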MaterialFunction /CosmeticCompanions/Assets/Quadruped/ShadowLap/Materials/MF_ShadowLap_Flipbook
graph {nodes [
  n1 = FunctionInput() {inputName=Material, inputType=FunctionInput_MaterialAttributes}
  n2 = Multiply(A=Subtract(A=WorldPosition().XYZ, B=ObjectPositionWS())) {constB=-1}
  n3 = Saturate(None=OneMinus(None=NamedRerouteUsage()))
  n4 = StaticSwitchParameter(
   True=NamedRerouteUsage()
   False=Add(
    A=GetMaterialAttributes(None=@n1) {attributeGetTypes [#1, #2]}.WorldPositionOffset
    B=Multiply(A=@n2, B=@n3)
   )
  ) {group=Flipbook Debug, parameterName=DebugVertexOffset, sortPriority=1}
  n5 = GetMaterialAttributes(None=StaticSwitchParameter(
    True=SetMaterialAttributes!(MaterialAttributes=@n1, WorldPositionOffset=@n4) {attributeSetTypes [#1]}
    False=@n1
   ) {group=Geometry Flipbook, parameterName=UseGeoFlipbook, sortPriority=-10}) {attributeGetTypes [#2, #3]}
  n6 = Constant() {r=1}
  n7 = Multiply(
   A=Constant3Vector() {constant=<rgba 1 1 1 0>}
   B=MaterialFunctionCall(Value_Morning=@n6, Value_Day=@n6, Value_Evening=@n6, Value_Night=Constant() {r=0.4}) {materialFunction=&Scalar_Time_of_Day}.Result
  )
  n8 = Constant() {r=1}
  n9 = If(A=@n5.BaseColor, "A > B"=@n8, "A == B"=@n8, "A < B"=Constant()) {constB=0.7}
  FunctionOutput(None=SetMaterialAttributes!(
    MaterialAttributes=@n5.MaterialAttributes
    EmissiveColor=LinearInterpolate(A=@n5.EmissiveColor, B=@n7, Alpha=@n9)
    BaseColor=LinearInterpolate(A=@n5.BaseColor, Alpha=@n9) {constB=0}
   ) {attributeSetTypes [#3, #2]})
  n10 = ChannelMaskParameter(None=VertexColor()) {group=Geometry Flipbook, parameterName=EyeVertexColorChannel}
  NamedRerouteDeclaration(None=LinearInterpolate(A=@n10.R, Alpha=Step(Y=@n10.R) {constX=0.01})) {name=EyeVertexColor, nodeColor=<rgba 0.8047 1 0 1>}
  NamedRerouteDeclaration(None=ScalarParameter() {group="Curve Parameters (Driven By Animation)", parameterName=R_brow_In_Dn_pose}) {name=GlabellaDownPose, nodeColor=<rgba 1 0.2105 0 1>}
  NamedRerouteDeclaration(None=ScalarParameter() {defaultValue=0.2, group=Geometry Flipbook, parameterName=VertColor_GlabellaDown}) {name=GlabellaDownVertColor, nodeColor=<rgba 0.2818 0.3488 1 1>}
  NamedRerouteDeclaration(None=ScalarParameter() {group="Curve Parameters (Driven By Animation)", parameterName=R_brow_Ot_Dn_pose}) {name=GlabellaUpPose, nodeColor=<rgba 1 0.2105 0 1>}
  NamedRerouteDeclaration(None=ScalarParameter() {defaultValue=0.1, group=Geometry Flipbook, parameterName=VertColor_GlabellaUp}) {name=GlabellaUpVertColor, nodeColor=<rgba 0.2818 0.3488 1 1>}
  NamedRerouteDeclaration(None=ScalarParameter() {defaultValue=0.7, group=Geometry Flipbook, parameterName="Threshold_LeftClosed (cheek up)"}) {name=Threshold_LeftClosed, nodeColor=<rgba 1 0.0834 0.4399 1>}
  n11 = Sign(None=ComponentMask(None=LocalPosition() {localOrigin=InstancePreSkinning}.XYZ) {r=true})
  n12 = MaterialFunctionCall!(Hide Element Toggle=Constant() {r=1}, VertexColorChannel=NamedRerouteUsage(), Vertex Value to hide=NamedRerouteUsage()) {materialFunction=/Game/Packages/Fortress_SharedMaterials/Base_Material_Functions/VertexColorMask}
  n13 = If(
   A=NamedRerouteUsage()
   B=NamedRerouteUsage()
   "A > B"=Multiply(A=OneMinus(None=@n11), B=@n12.Result)
   "A == B"=@n12.Result
   "A < B"=@n12.Result
  )
  n14 = If(
   A=NamedRerouteUsage()
   B=NamedRerouteUsage()
   "A > B"=MaterialFunctionCall(Hide Element Toggle=Constant() {r=1}, VertexColorChannel=NamedRerouteUsage(), Vertex Value to hide=NamedRerouteUsage()) {materialFunction=&VertexColorMask}.Result
   "A == B"=@n13
   "A < B"=@n13
  )
  n15 = If(
   A=NamedRerouteUsage()
   B=NamedRerouteUsage()
   "A > B"=MaterialFunctionCall(Hide Element Toggle=Constant() {r=1}, VertexColorChannel=NamedRerouteUsage(), Vertex Value to hide=NamedRerouteUsage()) {materialFunction=&VertexColorMask}.Result
   "A == B"=@n14
   "A < B"=@n14
  )
  n16 = If(
   A=NamedRerouteUsage()
   B=NamedRerouteUsage()
   "A > B"=MaterialFunctionCall(Hide Element Toggle=Constant() {r=1}, VertexColorChannel=NamedRerouteUsage(), Vertex Value to hide=NamedRerouteUsage()) {materialFunction=&VertexColorMask}.Result
   "A == B"=@n15
   "A < B"=@n15
  )
  n17 = If(
   A=NamedRerouteUsage()
   B=NamedRerouteUsage()
   "A > B"=MaterialFunctionCall(Hide Element Toggle=Constant() {r=1}, VertexColorChannel=NamedRerouteUsage(), Vertex Value to hide=NamedRerouteUsage()) {materialFunction=&VertexColorMask}.Result
   "A == B"=@n16
   "A < B"=@n16
  )
  n18 = If(
   A=NamedRerouteUsage()
   B=NamedRerouteUsage()
   "A > B"=MaterialFunctionCall(Hide Element Toggle=Constant() {r=1}, VertexColorChannel=NamedRerouteUsage(), Vertex Value to hide=NamedRerouteUsage()) {materialFunction=&VertexColorMask}.Result
   "A == B"=@n17
   "A < B"=@n17
  )
  n19 = If(
   A=NamedRerouteUsage()
   B=NamedRerouteUsage()
   "A > B"=MaterialFunctionCall(Hide Element Toggle=Constant() {r=1}, VertexColorChannel=NamedRerouteUsage(), Vertex Value to hide=NamedRerouteUsage()) {materialFunction=&VertexColorMask}.Result
   "A == B"=@n18
   "A < B"=@n18
  )
  n20 = If(
   A=NamedRerouteUsage()
   B=NamedRerouteUsage()
   "A > B"=MaterialFunctionCall(Hide Element Toggle=Constant() {r=1}, VertexColorChannel=NamedRerouteUsage(), Vertex Value to hide=NamedRerouteUsage()) {materialFunction=&VertexColorMask}.Result
   "A == B"=@n19
   "A < B"=@n19
  )
  NamedRerouteDeclaration(None=If(A=NamedRerouteUsage(), B=Constant() {r=0.95}, "A > B"=Constant() {r=1}, "A == B"=@n20, "A < B"=@n20)) {name=OutputPose, nodeColor=<rgba 0 0.1797 1 1>}
  NamedRerouteDeclaration(None=ScalarParameter() {defaultValue=0.3, group=Geometry Flipbook, parameterName=VertColor_DefaultPose}) {name=DefaultPoseVertColor, nodeColor=<rgba 0.2818 0.3488 1 1>}
  NamedRerouteDeclaration(None=ScalarParameter() {defaultValue=0.5, group=Geometry Flipbook, parameterName=VertColor_SquintPose}) {name=SquintPoseVertColor, nodeColor=<rgba 0.2818 0.3488 1 1>}
  NamedRerouteDeclaration(None=ScalarParameter() {defaultValue=0.6, group=Geometry Flipbook, parameterName=VertColor_BlinkPose}) {name=BlinkPoseVertColor, nodeColor=<rgba 0.2818 0.3488 1 1>}
  NamedRerouteDeclaration(None=ScalarParameter() {defaultValue=0.375, group=Geometry Flipbook, parameterName=Threshold_SquintInner}) {name=ThresholdSquintInner, nodeColor=<rgba 1 0.0834 0.4399 1>}
  NamedRerouteDeclaration(None=ScalarParameter() {group="Curve Parameters (Driven By Animation)", parameterName=R_blink_pose}) {name=BlinkPose, nodeColor=<rgba 1 0.2105 0 1>}
  NamedRerouteDeclaration(None=ScalarParameter() {group="Curve Parameters (Driven By Animation)", parameterName=R_lower_lid_up_pose}) {name=SquintInnerPose, nodeColor=<rgba 1 0.2105 0 1>}
  NamedRerouteDeclaration(None=ScalarParameter() {group="Curve Parameters (Driven By Animation)", parameterName=R_wide_pose}) {name=WidePose, nodeColor=<rgba 1 0.2105 0 1>}
  NamedRerouteDeclaration(None=ScalarParameter() {defaultValue=0.7, group=Geometry Flipbook, parameterName=Threshold_EyeWide}) {name=Threshold_EyeWide, nodeColor=<rgba 1 0.0834 0.4399 1>}
  NamedRerouteDeclaration(None=ScalarParameter() {defaultValue=0.5, group=Geometry Flipbook, parameterName=VertColor_EyeWide}) {name=EyeWideVertColor, nodeColor=<rgba 0.2818 0.3488 1 1>}
  NamedRerouteDeclaration(None=ScalarParameter() {group="Curve Parameters (Driven By Animation)", parameterName=R_brow_squeeze_in_pose}) {name=SqueezePose, nodeColor=<rgba 1 0.2105 0 1>}
  NamedRerouteDeclaration(None=ScalarParameter() {defaultValue=0.7, group=Geometry Flipbook, parameterName=Threshold_EyeSqueeze}) {name=Threshold_EyeSqueeze, nodeColor=<rgba 1 0.0834 0.4399 1>}
  NamedRerouteDeclaration(None=ScalarParameter() {defaultValue=0.5, group=Geometry Flipbook, parameterName=VertColor_EyeSqueeze}) {name=EyeSqueezeVertColor, nodeColor=<rgba 0.2818 0.3488 1 1>}
  NamedRerouteDeclaration(None=ScalarParameter() {defaultValue=0.1, group=Geometry Flipbook, parameterName=Threshold_BrowDown}) {name=ThresholdBrowDown, nodeColor=<rgba 1 0.0834 0.4399 1>}
  NamedRerouteDeclaration(None=ScalarParameter() {defaultValue=0.1, group=Geometry Flipbook, parameterName=Threshold_GlabellaDown}) {name=ThresholdGlabellaDown, nodeColor=<rgba 1 0.0834 0.4399 1>}
  NamedRerouteDeclaration(None=ScalarParameter() {defaultValue=0.1, group=Geometry Flipbook, parameterName=Threshold_GlabellaUp}) {name=ThresholdGlabellaUp, nodeColor=<rgba 1 0.0834 0.4399 1>}
  NamedRerouteDeclaration(None=ScalarParameter() {defaultValue=0.1, group=Geometry Flipbook, parameterName=Threshold_Blink}) {name=ThresholdBlink, nodeColor=<rgba 1 0.0834 0.4399 1>}
  NamedRerouteDeclaration(None=ScalarParameter() {group="Curve Parameters (Driven By Animation)", parameterName=R_cheek_up_pose}) {name=LeftClosed, nodeColor=<rgba 1 0.2105 0 1>}
  NamedRerouteDeclaration(None=ScalarParameter() {group="Curve Parameters (Driven By Animation)", parameterName=R_brow_down_pose}) {name=BrowDownPose, nodeColor=<rgba 1 0.2105 0 1>}
  n21 = ChannelMaskParameter(None=VertexColor()) {group=Flipbook 06 Debug, maskChannel=Green, parameterName=DebugVertColorChannel, sortPriority=21, defaultValue=<rgba 0 1 0 0>}
  NamedRerouteDeclaration(None=Multiply(A=Step(X=@n21.R) {constY=0.025}, B=Step(Y=@n21.R) {constX=0.975})) {name=DebugVertexMask, nodeColor=<rgba 0.1953 1 0 1>}
  n22 = NamedRerouteUsage()
  n23 = NamedRerouteUsage()
  n24 = Subtract(A=Multiply(A=Fmod(A=@n22, B=@n23), B=NamedRerouteUsage()) {constB=50}, B=NamedRerouteUsage())
  n25 = Constant()
  n26 = Divide(A=NamedRerouteUsage(), B=NamedRerouteUsage()) {constB=3}
  n27 = Subtract(A=Multiply(A=NamedRerouteUsage(), B=Floor(None=@n26)), B=NamedRerouteUsage())
  n28 = VectorParameter() {group=Flipbook Debug, parameterName=FB_DebugVertColorOffset, sortPriority=23, defaultValue=<rgba 0 0 0 1>}
  NamedRerouteDeclaration(None=Add(
    A=MaterialFunctionCall(X=@n24, Y=@n25, Z=@n27) {materialFunction=&MakeFloat3}.Result
    B=@n28.RGB
   )) {name=DebugVertexOffset, nodeColor=<rgba 0 0.9832 1 1>}
  NamedRerouteDeclaration(None=ScalarParameter() {defaultValue=20, group=Flipbook Debug, parameterName=FB_DebugVertColor_CellSize, sortPriority=22}) {name=DebugVertColorCellSize, nodeColor=<rgba 1 0.8438 0 1>}
  NamedRerouteDeclaration(None=Constant() {r=3}) {name=DebugVertColorColumnCount, nodeColor=<rgba 1 0.2578 0 1>}
  n29 = Subtract(A=NamedRerouteUsage())
  n30 = NamedRerouteUsage()
  NamedRerouteDeclaration(None=Multiply(A=Multiply(A=@n29, B=@n30)) {constB=0.5}) {name=DebugVertColorOffset, nodeColor=<rgba 0.7109 0 1 1>}
  n31 = Multiply(A=@n21.R) {constB=20}
  NamedRerouteDeclaration(None=Round(None=Subtract(A=@n31) {constB=2})) {name=DebugVertColorIndex, nodeColor=<rgba 0 1 0.3906 1>}
  NamedRerouteDeclaration(None=ScalarParameter() {defaultValue=0.4, group=Geometry Flipbook, parameterName=VertColor_BrowDown}) {name=BrowDownVertColor, nodeColor=<rgba 0.2818 0.3488 1 1>}
 ]}
