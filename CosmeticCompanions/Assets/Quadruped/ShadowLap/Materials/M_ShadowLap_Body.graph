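# Body material for the ShadowLap quadruped cosmetic companion, built as a material-attributes stack on the 2023 Character Parent functions.
# The parent PreFX function feeds companion custom color, then basic tech-art customization, with PreFX also supplying mask data.
# ShadowLap-specific Body and Flipbook functions follow, before the shared PostFX function finishes the chain.
# Shading model comes from the graph, with skeletal mesh, morph target, mesh deformer and Nanite usage enabled.
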Material /CosmeticCompanions/Assets/Quadruped/ShadowLap/Materials/M_ShadowLap_Body
graph {
 properties {
  bUseMaterialAttributes=true
  bUsedWithMeshDeformer=true
  bUsedWithMorphTargets=true
  bUsedWithNanite=true
  bUsedWithSkeletalMesh=true
  pixelDepthOffsetMode=PDOM_Legacy
  shadingModel=MSM_FromMaterialExpression
  usedShadingModels=Default Lit | Subsurface | Clear Coat | Subsurface Profile | Cloth
 }
 nodes [
  n1 = MaterialFunctionCall!(Overide Texture Set UVs, Override Tech Art Mask UVs, Override Basic Emissive UVs) {materialFunction=/Game/Global/Materials/Character_Parent_2023/Functions/MF_CharacterParent_2023_Customization_PreFX}
  n2 = MaterialFunctionCall!(
   Tech Art Mask=@n1.Tech Art Mask
   Emissive UVs=@n1.Tech Art Mask
   LocalPosition=@n1.Tech Art Mask
   Fresnel=@n1.Tech Art Mask
   Mask Input=@n1.Tech Art Mask
   Material=MaterialFunctionCall!(Material=@n1.Result, Tech Art Mask=@n1.Result) {materialFunction=/CosmeticCompanions/Materials/MaterialFunctions/MF_FN_Companion_Custom_Color}.Base Color
  ) {materialFunction=/Game/Global/Materials/Character_Parent_2023/Functions/MF_CharacterParent_2023_CustomizationBasicTechArt}
  MaterialFunctionCall(
   Material=MaterialFunctionCall!(Material=MaterialFunctionCall!(Result=@n2.Result) {materialFunction=&/MF_ShadowLap_Body}.SetMaterialAttributes) {materialFunction=&/MF_ShadowLap_Flipbook}.Result
   M=@n1.M
   Pre Skinned Local Position=@n1.M
   Distance Blend=@n1.M
   Fresnel=@n1.M
  ) {materialFunction=&MF_CharacterParent_2023_PostFX}
 ]
}
