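# Post-process AnimBP for the ShadowLap quadruped cosmetic companion on the WideSnore skeleton, applying facial poses over the incoming animation.
# The input pose runs through the head facial PoseAsset blend node straight to the output.
# A chain of spring-based, world-space Anim Dynamics nodes for left and right ear bones sits unconnected to the output, with every alpha at zero.
# The event graph holds only unused default nodes.
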AnimBlueprint /CosmeticCompanions/Assets/Quadruped/ShadowLap/Meshes/ShadowLap_Quadruped_Post_AnimBP
target_skeleton=/CosmeticCompanions/Assets/Quadruped/WideSnore/Meshes/WideSnore_Quadruped_Skeleton
graphs [
 AnimGraph {nodes [
   n1 = Anim Dynamics - Bone: ear_02_l(ComponentPose=Local To Component(), Alpha=0, ExternalForce="0,0,0") {
    angularDampingOverride=0.94
    angularSpringConstant=5000
    bAngularSpring=true
    bOverrideAngularDamping=true
    bUseGravityOverride=true
    simulationSpace=World
    alphaBoolBlend {blendOption=Linear}
    boundBone {boneName=ear_02_l}
    componentAppliedLinearAccClamp=<500 500 500>
    gravityOverride {z=-490}
    physicsBodyDefinitions [{
      boundBone {boneName=ear_02_l}
      boxExtents=<10 10 10>
      constraintSetup {angularLimitsMax=<0 0 0>, angularLimitsMin=<0 0 0>, angularTarget=<0 0 0>}
      localJointOffset=<0 0 0>
     }]
    relativeSpaceBone {boneName=spine_03}
   }
   n2 = Anim Dynamics - Bone: ear_03_l(ComponentPose=@n1, Alpha=0, ExternalForce="0,0,0") {
    angularDampingOverride=0.94
    angularSpringConstant=5000
    bAngularSpring=true
    bOverrideAngularDamping=true
    bUseGravityOverride=true
    simulationSpace=World
    alphaBoolBlend {blendOption=Linear}
    boundBone {boneName=ear_03_l}
    componentAppliedLinearAccClamp=<500 500 500>
    gravityOverride {z=-490}
    physicsBodyDefinitions [{
      boundBone {boneName=ear_03_l}
      boxExtents=<10 10 10>
      constraintSetup {angularLimitsMax=<0 0 0>, angularLimitsMin=<0 0 0>, angularTarget=<0 0 0>}
      localJointOffset=<0 0 0>
     }]
    relativeSpaceBone {boneName=spine_03}
   }
   n3 = Anim Dynamics - Bone: ear_04_l(ComponentPose=@n2, Alpha=0, ExternalForce="0,0,0") {
    angularDampingOverride=0.94
    angularSpringConstant=5000
    bAngularSpring=true
    bOverrideAngularDamping=true
    bUseGravityOverride=true
    simulationSpace=World
    alphaBoolBlend {blendOption=Linear}
    boundBone {boneName=ear_04_l}
    componentAppliedLinearAccClamp=<500 500 500>
    gravityOverride {z=-490}
    physicsBodyDefinitions [{
      boundBone {boneName=ear_04_l}
      boxExtents=<10 10 10>
      constraintSetup {angularLimitsMax=<0 0 0>, angularLimitsMin=<0 0 0>, angularTarget=<0 0 0>}
      localJointOffset=<0 0 0>
     }]
    relativeSpaceBone {boneName=spine_03}
   }
   n4 = Anim Dynamics - Bone: ear_02_r(ComponentPose=@n3, Alpha=0, ExternalForce="0,0,0") {
    angularDampingOverride=0.94
    angularSpringConstant=5000
    bAngularSpring=true
    bOverrideAngularDamping=true
    bUseGravityOverride=true
    simulationSpace=World
    alphaBoolBlend {blendOption=Linear}
    boundBone {boneName=ear_02_r}
    componentAppliedLinearAccClamp=<500 500 500>
    gravityOverride {z=-490}
    physicsBodyDefinitions [{
      boundBone {boneName=ear_02_r}
      boxExtents=<10 10 10>
      constraintSetup {angularLimitsMax=<0 0 0>, angularLimitsMin=<0 0 0>, angularTarget=<0 0 0>}
      localJointOffset=<0 0 0>
     }]
    relativeSpaceBone {boneName=spine_03}
   }
   n5 = Anim Dynamics - Bone: ear_03_r(ComponentPose=@n4, Alpha=0, ExternalForce="0,0,0") {
    angularDampingOverride=0.94
    angularSpringConstant=5000
    bAngularSpring=true
    bOverrideAngularDamping=true
    bUseGravityOverride=true
    simulationSpace=World
    alphaBoolBlend {blendOption=Linear}
    boundBone {boneName=ear_03_r}
    componentAppliedLinearAccClamp=<500 500 500>
    gravityOverride {z=-490}
    physicsBodyDefinitions [{
      boundBone {boneName=ear_03_r}
      boxExtents=<10 10 10>
      constraintSetup {angularLimitsMax=<0 0 0>, angularLimitsMin=<0 0 0>, angularTarget=<0 0 0>}
      localJointOffset=<0 0 0>
     }]
    relativeSpaceBone {boneName=spine_03}
   }
   n6 = Anim Dynamics - Bone: ear_04_r(ComponentPose=@n5, Alpha=0, ExternalForce="0,0,0") {
    angularDampingOverride=0.94
    angularSpringConstant=5000
    bAngularSpring=true
    bOverrideAngularDamping=true
    bUseGravityOverride=true
    simulationSpace=World
    alphaBoolBlend {blendOption=Linear}
    boundBone {boneName=ear_04_r}
    componentAppliedLinearAccClamp=<500 500 500>
    gravityOverride {z=-490}
    physicsBodyDefinitions [{
      boundBone {boneName=ear_04_r}
      boxExtents=<10 10 10>
      constraintSetup {angularLimitsMax=<0 0 0>, angularLimitsMin=<0 0 0>, angularTarget=<0 0 0>}
      localJointOffset=<0 0 0>
     }]
    relativeSpaceBone {boneName=spine_03}
   }
   Component To Local(ComponentPose=@n6)
   Output Pose(Result=ShadowLap_Quadruped_Head_Facial_Poses_PoseAsset!(SourcePose:Pose Link Structure=Input Pose()) -> (Pose:Pose Link Structure) {
     is=title AnimGraphNode_PoseBlendNode
     blendOption=Linear
     poseAsset=&/ShadowLap_Quadruped_Head_Facial_Poses_PoseAsset
    })
  ]}
 EventGraph {nodes [
   TryGetPawnOwner()
   Event BlueprintUpdateAnimation()
  ]}
]
variables []
